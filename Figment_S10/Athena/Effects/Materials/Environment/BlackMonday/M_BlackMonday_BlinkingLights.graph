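# Translucent, unlit, two-sided material for mesh particles that provides blinking lights for the BlackMonday environment effects.
# The U coordinate is squared, scaled, rounded and saturated into a hard binary mask that splits the UV space into two sets of lights.
# The SwitchLight dynamic parameter blends between that mask and its inverse, so the particle system controls which set is lit, and particle alpha scales the result.
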Material /Figment_S10/Athena/Effects/Materials/Environment/BlackMonday/M_BlackMonday_BlinkingLights
graph {
 properties {
  bUsedWithMeshParticles=true
  blendMode=BLEND_Translucent
  pixelDepthOffsetMode=PDOM_Legacy
  shadingModel=MSM_Unlit
  twoSided=true
  usedShadingModels=Unlit
 }
 nodes [
  n1 = ComponentMask(None=TextureCoordinate() {uTiling=0.5}) {r=true}
  n2 = Multiply(A=Multiply(A=@n1, B=@n1)) {constB=10}
  n3 = Saturate(None=Ceil(None=Round(None=@n2)))
  Multiply(
   A=ParticleColor().A
   B=LinearInterpolate(A=@n3, B=OneMinus(None=@n3), Alpha=DynamicParameter() {paramNames [SwitchLight,Param2,Param3,Param4]}.SwitchLight)
  )
 ]
}
