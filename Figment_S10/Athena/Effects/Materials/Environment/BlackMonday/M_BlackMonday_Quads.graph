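# Translucent, two-sided lit particle sprite material for BlackMonday environment effects, drawing tumbling quads with randomized color.
# A vertex rotation spins each sprite about a randomly picked axis, with the angle driven by particle distance from world origin, and is zeroed on ES2 and ES3.1.
# Particle alpha and contrast-adjusted noise feed an erosion function for opacity.
# Particle color gets per-particle random brightness and hue shift.
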Material /Figment_S10/Athena/Effects/Materials/Environment/BlackMonday/M_BlackMonday_Quads
graph {
 properties {
  bEnableResponsiveAA=true
  bUsedWithParticleSprites=true
  blendMode=BLEND_Translucent
  opacityMaskClipValue=0.1
  pixelDepthOffsetMode=PDOM_Legacy
  translucencyLightingMode=TLM_VolumetricPerVertexDirectional
  twoSided=true
  usedShadingModels=Default Lit
 }
 nodes [
  n1 = ParticleRandom()
  n2 = VectorParameter() {parameterName=NormalizeRotationAxis_1, defaultValue=<rgba 1 0.3 -0.2 0>}
  n3 = VectorParameter() {parameterName=NormalizeRotationAxis_2, defaultValue=<rgba -0.44 1 0.13 0>}
  n4 = VectorParameter() {parameterName=NormalizeRotationAxis_3, defaultValue=<rgba 0.06 -0.22 1 0>}
  n5 = Distance(A=Constant3Vector(), B=ParticlePositionWS())
  n6 = ScalarParameter() {defaultValue=100, parameterName=Spin Scale}
  n7 = ParticleRandom()
  n8 = RotateAboutAxis(
   NormalizedRotationAxis=Normalize(VectorInput=MaterialFunctionCall!(Random Picker=@n1, Texture Input, Option 1=@n2.RGB, Option 2=@n3.RGB, Option 3=@n4.RGB) {materialFunction=/Game/Effects/Fort_Effects/Materials/_Functions/MF_RGB_Random_01}.Vec3 Out Only)
   RotationAngle=Multiply(A=Divide(A=@n5, B=@n6), B=Clamp(None=@n7) {minDefault=0.55})
   PivotPoint=ParticlePositionWS()
   Position=WorldPosition().XYZ
  )
  n9 = Constant()
  FeatureLevelSwitch(Default=@n8, ES2=@n9, ES3_1=@n9, SM4_REMOVED=@n8)
  n10 = ParticleColor()
  n11 = TextureSample(UVs=TextureCoordinate() {uTiling=0.25, vTiling=0.25}) {samplerType=SAMPLERTYPE_Masks, texture=/Game/Weapons/WeaponSkins/Wraps/Textures/Mash/Mash_Noise}
  n12 = OneMinus(None=@n10.A)
  MaterialFunctionCall!(
   Dynamic Param=@n10.A
   Alpha=Multiply(
    A=@n10.A
    B=MaterialFunctionCall(In=@n11.G, Contrast=@n12) {materialFunction=&CheapContrast}.Result
   )
  ) {materialFunction=/Game/Effects/Fort_Effects/Materials/_Functions/Erode_Alpha}
  n13 = ParticleRandom()
  MaterialFunctionCall(Hue Shift Percentage=LinearInterpolate(Alpha=ParticleRandom()), Texture=Multiply(A=LinearInterpolate(Alpha=@n13) {constA=0.4}, B=@n10.RGB)) {materialFunction=&HueShift}
 ]
}
