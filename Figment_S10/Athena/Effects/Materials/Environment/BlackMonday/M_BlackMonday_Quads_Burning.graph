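# Translucent unlit two-sided particle material for Black Monday effects that draws burning V-Buck bills.
# A tinted, randomly hue-shifted bill texture, scaled by particle red, adds to panning noise embers masked by charcoal and scaled by particle green.
# Particle alpha drives a noise-based erode for burn-away.
# Quads spin about a random axis by particle world position; mobile disables the spin and uses dimmer embers.
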Material /Figment_S10/Athena/Effects/Materials/Environment/BlackMonday/M_BlackMonday_Quads_Burning
graph {
 properties {
  bEnableResponsiveAA=true
  bUsedWithParticleSprites=true
  blendMode=BLEND_Translucent
  opacityMaskClipValue=0.1
  pixelDepthOffsetMode=PDOM_Legacy
  shadingModel=MSM_Unlit
  translucencyLightingMode=TLM_VolumetricPerVertexDirectional
  twoSided=true
  usedShadingModels=Unlit
 }
 nodes [
  n1 = LinearInterpolate(Alpha=ParticleRandom()) {constB=0.15}
  n2 = ParticleColor()
  n3 = Multiply(
   A=LinearInterpolate(Alpha=ParticleRandom()) {constA=0.4}
   B=Multiply(A=VectorParameter() {parameterName=MoneyColor, defaultValue=<rgba 0.1251 0.2453 0.3 1>}.RGB, B=@n2.R)
  )
  n4 = Multiply(A=TextureSample(UVs=TextureCoordinate() {uTiling=0.25, vTiling=0.25}) {texture=/Game/Packages/Fortress_ShaderTests/Textures/charcoal}.G) {constB=4}
  n5 = ParticleRandom()
  n6 = ScalarParameter() {defaultValue=0.25, parameterName=Offset}
  n7 = Add(A=TextureCoordinate() {uTiling=0.05, vTiling=0.05}, B=ComponentMask(None=Divide(A=@n5, B=@n6)) {r=true})
  n8 = VectorParameter() {parameterName=NoiseTile, defaultValue=<rgba 3 3 6 6>}
  n9 = Panner(
   Coordinate=Multiply(A=@n7, B=ComponentMask(None=@n8.RGB) {g=true, r=true})
   Speed=AppendVector(A=ScalarParameter() {defaultValue=0.25, parameterName=MaskSpeed}, B=Constant())
  )
  n10 = Multiply(
   A=Saturate(None=Power(Base=@n4))
   B=Power(
    Base=TextureSample(UVs=@n9) {samplerType=SAMPLERTYPE_Grayscale, texture=/Figment_S10/Effects/Fort_Effects/Textures/Tile/TilingNoise06}.G
    Exp=ScalarParameter() {defaultValue=2, parameterName=MaskExp}
   )
  )
  n11 = AppendVector(A=@n8.B, B=@n8.B)
  n12 = ScalarParameter() {defaultValue=0.5, parameterName=NoiseSpeed}
  n13 = Constant() {r=0.25}
  n14 = TextureSample(UVs=Panner(Coordinate=Multiply(A=@n7, B=@n11), Speed=AppendVector(A=@n12, B=@n13))) {texture=/Game/Effects/Fort_Effects/Textures/Tile/TankMoteNoise}
  n15 = ParticleColor()
  n16 = VectorParameter() {parameterName=EmberColor, defaultValue=<rgba 0 100 30.1568 0>}
  n17 = VectorParameter() {parameterName=Mobile_EmberColor, defaultValue=<rgba 0 5 1.5078 0>}
  n18 = FeatureLevelSwitch(Default=@n16.RGB, ES2=@n17.RGB, ES3_1=@n17.RGB, SM4_REMOVED=@n16.RGB)
  Add(
   A=Multiply(
    A=TextureSample() {texture=/Game/Effects/Fort_Effects/Textures/Emotes/V_Buck_Bill_v2}.G
    B=MaterialFunctionCall(Hue Shift Percentage=@n1, Texture=@n3) {materialFunction=&HueShift}.Result
   )
   B=Multiply(A=Multiply(A=@n10, B=@n14.G), B=Multiply(A=@n15.G, B=@n18))
  )
  Constant() {r=1}
  n19 = ParticleRandom()
  n20 = VectorParameter() {parameterName=NormalizeRotationAxis_1, defaultValue=<rgba 1 0.3 -0.2 0>}
  n21 = VectorParameter() {parameterName=NormalizeRotationAxis_2, defaultValue=<rgba -0.44 1 0.13 0>}
  n22 = VectorParameter() {parameterName=NormalizeRotationAxis_3, defaultValue=<rgba 0.06 -0.22 1 0>}
  n23 = Distance(A=Constant3Vector(), B=ParticlePositionWS())
  n24 = ScalarParameter() {defaultValue=100, parameterName=Spin Scale}
  n25 = ParticleRandom()
  n26 = RotateAboutAxis(
   NormalizedRotationAxis=Normalize(VectorInput=MaterialFunctionCall!(Random Picker=@n19, Texture Input, Option 1=@n20.RGB, Option 2=@n21.RGB, Option 3=@n22.RGB) {materialFunction=/Game/Effects/Fort_Effects/Materials/_Functions/MF_RGB_Random_01}.Vec3 Out Only)
   RotationAngle=Multiply(A=Divide(A=@n23, B=@n24), B=Clamp(None=@n25) {minDefault=0.55})
   PivotPoint=ParticlePositionWS()
   Position=WorldPosition().XYZ
  )
  n27 = Constant()
  FeatureLevelSwitch(Default=@n26, ES2=@n27, ES3_1=@n27, SM4_REMOVED=@n26)
  n28 = TextureSample(UVs=TextureCoordinate() {uTiling=0.25, vTiling=0.25}) {samplerType=SAMPLERTYPE_Masks, texture=/Game/Weapons/WeaponSkins/Wraps/Textures/Mash/Mash_Noise}
  n29 = OneMinus(None=@n2.A)
  MaterialFunctionCall!(
   Dynamic Param=@n2.A
   Alpha=Multiply(
    A=@n2.A
    B=MaterialFunctionCall(In=@n28.G, Contrast=@n29) {materialFunction=&CheapContrast}.Result
   )
  ) {materialFunction=/Game/Effects/Fort_Effects/Materials/_Functions/Erode_Alpha}
  ObjectPositionWS()
  OneMinus(None=@n15.G)
  ScalarParameter() {defaultValue=1, parameterName="Contrast (S)"}
 ]
}
